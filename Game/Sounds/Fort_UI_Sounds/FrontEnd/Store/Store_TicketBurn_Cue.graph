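# Sound cue for the ticket burn sound in the front-end store, played on the UI sound class at half volume.
# A quality-level node selects the standard wave for its first level and a shared "opt" variant for the other three.
# Each wave passes through a default modulator for slight variation.
# The cue references the Monster Smasher attenuation asset despite being a UI sound.
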SoundCue /Game/Sounds/Fort_UI_Sounds/FrontEnd/Store/Store_TicketBurn_Cue
root=QualityLevel_0
nodes [
 {id=Modulator_0, children [WavePlayer_0]}
 {id=Modulator_1, children [WavePlayer_4]}
 {id=QualityLevel_0, children [Modulator_0,Modulator_1,Modulator_1,Modulator_1]}
 {id=WavePlayer_0, properties {soundWaveAssetPtr=/Game/Sounds/Fort_UI_Sounds/FrontEnd/Store/wav/fort_ticket_burn_01}}
 {id=WavePlayer_4, properties {soundWaveAssetPtr=/Game/Sounds/Fort_UI_Sounds/FrontEnd/Store/wav/fort_ticket_burn_01_opt}}
]
properties {
 attenuationSettings=/Game/Sounds/Attenuation/Fort_Monster_Smasher_Att
 maxDistance=7000
 soundClassObject=&Fort_UI
 volumeMultiplier=0.5
}
