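# Front-end UI sound cue for a Showdown point, built from two layered point sounds.
# A Mixer plays the first wave immediately at a raised fixed pitch and adds the second, quieter wave after a fixed short delay.
# A root Modulator applies slight random pitch and volume variation to the pair.
# Overridden concurrency rejects new instances at the limit, and the assigned sound class is Fort_Enemy_Damaged_PSM.
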SoundCue /Game/Sounds/Fort_UI_Sounds/FrontEnd/Store/Fort_UI_Showdown_Point_Cue
root=Modulator_0
nodes [
 {id=Delay_0, children [WavePlayer_2], properties {delayMax=0.66, delayMin=0.66}}
 {id=Mixer_0, children [Modulator_3,Delay_0], properties {inputVolume [1,0.4]}}
 {id=Modulator_0, children [Mixer_0], properties {volumeMax=1}}
 {id=Modulator_3, children [WavePlayer_1], properties {pitchMax=1.8, pitchMin=1.8, volumeMax=1, volumeMin=1}}
 {id=WavePlayer_1, properties {soundWaveAssetPtr=/Game/Sounds/Fort_UI_Sounds/FrontEnd/UI_Showdown_Point_01}}
 {id=WavePlayer_2, properties {soundWaveAssetPtr=/Game/Sounds/Fort_UI_Sounds/FrontEnd/UI_Showdown_Point_02}}
]
properties {
 bOverrideConcurrency=true
 maxDistance=2097152
 soundClassObject=/Game/Sounds/FortSoundClassesAndModes/Fort_Enemy_Damaged_PSM
 volumeMultiplier=0.7
 concurrencyOverrides {resolutionRule=PreventNew}
}
